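# Widget that displays the logo for an event countdown timer as a CMS-loaded image in a fill-scaled box.
# On Construct it reads the CountdownVisuals row of DT_CountdownVisuals and passes only its LogoURL to the image.
# UpdateOpacity and SetMediaURL let callers change opacity or URL, and a transparent blank texture is the default and load-failure fallback.
# The Languages and Logo URLs arrays are not used by these graphs.
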WidgetBlueprint /Game/Athena/Events/Generic/CountdownTimer/Widgets/CountdownLogoWidget
graphs [
 EventGraph {nodes [
   n1 = UpdateOpacity!() -> (OutputDelegate, then, InOpacity:float=~) {is=title K2Node_CustomEvent}
   > SetOpacity!(execute, self:Image Object Reference=Get Image() -> (Image:Epic CMSImage Object Reference), InOpacity:float=@n1.InOpacity) -> (then)
   Event Construct()
   > n2 = Get Data Table Row DT_CountdownVisuals!(
    execute:Exec
    DataTable:Data Table Object Reference=/Game/Athena/Events/Generic/CountdownTimer/DataTables/DT_CountdownVisuals
    RowName:Name=CountdownVisuals
   ) -> (
    then:Exec
    RowNotFound:Exec
    ReturnValue_FrontendTimerEnabled_6_#1:bool=~
    ReturnValue_ShowBeginningSoonText_31_#2:bool=~
    ReturnValue_ShowLogo_28_#3:bool=~
    ReturnValue_LogoURL_34_#4:String
    ReturnValue_LogoTransform_29_#5:Transform
    ReturnValue_Timer_Color_Numbers_13_#6:Linear Color Structure
    ReturnValue_Timer_Color_Text_15_#7:Linear Color Structure
    ReturnValue_Timer_Color_Image_17_#8:Linear Color Structure
    ReturnValue_Timer_Color_Background_19_#9:Linear Color Structure
   ) {is=title K2Node_GetDataTableRow}
   > SetMediaURL!(execute, self:Epic CMSImage Object Reference=Get Image(), MediaUrl:String=@n2.ReturnValue_LogoURL_34_#4) -> (then)
  ]}
 SetMediaURL {nodes [
   n1 = SetMediaURL!() -> (then, MediaUrl:String) {is=title K2Node_FunctionEntry}
   > SetMediaURL(self=Get Image(), MediaUrl=@n1) {is=title K2Node_CallFunction}
  ]}
]
variables [
 Languages {type="array<string>", default [en,fr,de,es,it,pt,ru,pl,ja,ko,zh,ar,nl,sv,tr,cs,da,fi,no,hu]}
 Logo URLs {type="array<string>", default [""]}
]
widgets {
 info {inheritedWidgetCount=0, parentClass=&UserWidget, rootWidgetClass=&ScaleBox, widgetCount=2}
 widgets [
  ScaleBox_0 {widgetClassPath=&ScaleBox, widget=ScaleBox_0, properties {stretch=ScaleToFill}}
  Image {
   widgetClassPath=/Script/EpicCMSUIFramework.EpicCMSImage
   bIsVariable=true
   parent=ScaleBox_0
   widget=Image
   properties {
    brush {
     bIsDynamicallyLoaded=false
     drawAs=Image
     imageType=FullColor
     mirroring=NoMirror
     resourceName=None
     resourceObject=/Game/Global/Textures/Default/Blanks/T_Blank_White_ZeroAlpha_sRGBColor
     tiling=NoTile
     imageSize=<1 1>
     margin=<ltrb 0 0 0 0>
     outlineSettings {
      bUseBrushTransparency=false
      roundingType=HalfHeightRadius
      width=0
      color {colorUseRule=UseColor_Specified, specifiedColor=<rgba 0 0 0 0>}
      cornerRadii=<0 0 0 0>
     }
     tintColor {colorUseRule=UseColor_Specified, specifiedColor=<rgba 1 1 1 1>}
     uVRegion {bIsValid=false, max=<0 0>, min=<0 0>}
    }
    loadingBackgroundBrush {
     bIsDynamicallyLoaded=false
     drawAs=NoDrawType
     imageType=NoImage
     mirroring=NoMirror
     resourceName=None
     resourceObject=null
     tiling=NoTile
     imageSize=<16 9>
     margin=<ltrb 0 0 0 0>
     outlineSettings {
      bUseBrushTransparency=false
      roundingType=HalfHeightRadius
      width=0
      color {colorUseRule=UseColor_Specified, specifiedColor=<rgba 0 0 0 0>}
      cornerRadii=<0 0 0 0>
     }
     tintColor {colorUseRule=UseColor_Specified, specifiedColor=<rgba 1 1 1 1>}
     uVRegion {bIsValid=false, max=<0 0>, min=<0 0>}
    }
    loadingFailFallback {
     bIsDynamicallyLoaded=false
     drawAs=Image
     imageType=NoImage
     mirroring=NoMirror
     resourceName=None
     resourceObject=&T_Blank_White_ZeroAlpha_sRGBColor
     tiling=NoTile
     imageSize=<1 1>
     margin=<ltrb 0 0 0 0>
     outlineSettings {
      bUseBrushTransparency=false
      roundingType=HalfHeightRadius
      width=0
      color {colorUseRule=UseColor_Specified, specifiedColor=<rgba 0 0 0 0>}
      cornerRadii=<0 0 0 0>
     }
     tintColor {colorUseRule=UseColor_Specified, specifiedColor=<rgba 1 1 1 1>}
     uVRegion {bIsValid=false, max=<0 0>, min=<0 0>}
    }
   }
   slot {class=ScaleBoxSlot}
  }
 ]
}
